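# Sound cue for the activation sound of Slone's MultiLinker, apparently an ability of the Slone NPC.
# A single wave, MirrorImage_Activate_01, plays through a modulator that randomizes pitch for variation, with a boosted cue volume.
# It uses Slone's ability attenuation and a melee weapon sound class, overrides concurrency to cap simultaneous instances at two, and primes on load.
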SoundCue /NPCLibrary/Audio/Slone/Slone_MultiLinker_Activate_Cue
root=Modulator_0
nodes [
 {id=Modulator_0, children [WavePlayer_0], properties {pitchMax=1.1}}
 {id=WavePlayer_0, properties {soundWaveAssetPtr=&/MirrorImage_Activate_01}}
]
properties {
 attenuationSettings=&/Slone_Ability_Att
 bOverrideConcurrency=true
 bPrimeOnLoad=true
 maxDistance=4000
 soundClassObject=&Fort_WeaponMelee_PSM
 volumeMultiplier=1.5
 concurrencyOverrides {maxCount=2}
}
