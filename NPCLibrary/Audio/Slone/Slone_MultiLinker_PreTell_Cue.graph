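# Sound cue for the pre-tell of Slone's MultiLinker, an NPC ability warning sound.
# A single wave player feeds the output directly, playing the MirrorImage_PreTell_01 wave with no randomization or layering.
# It uses Slone's ability attenuation and the Fort_WeaponMelee_PSM sound class, with a slightly boosted volume.
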SoundCue /NPCLibrary/Audio/Slone/Slone_MultiLinker_PreTell_Cue
root=WavePlayer_0
nodes [{id=WavePlayer_0, properties {soundWaveAssetPtr=&/MirrorImage_PreTell_01}}]
properties {
 attenuationSettings=&/Slone_Ability_Att
 maxDistance=4000
 soundClassObject=&Fort_WeaponMelee_PSM
 volumeMultiplier=1.2
}
